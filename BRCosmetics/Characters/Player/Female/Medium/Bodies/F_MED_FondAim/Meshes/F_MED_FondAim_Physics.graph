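# Physics asset for the F_MED_FondAim female medium player body cosmetic, with capsule bodies for the torso, limbs and a hanging belt chain.
# A tiny based_belt_01 anchor locks angularly to the pelvis and carries belt_01, belt_02 and belt_04 under tight limits.
# Spine_01 and the twist-bone thighs swing freely from the pelvis; calves also constrain directly to it.
# The arms are asymmetric: the left lacks a forearm body and has a tiny upper-arm capsule.
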PhysicsAsset /BRCosmetics/Characters/Player/Female/Medium/Bodies/F_MED_FondAim/Meshes/F_MED_FondAim_Physics
bodies {
 based_belt_01 [{shapeType=Capsule, length=0.505, radius=0.505, shapeName=based_belt_01_capsule, rotation=<rot 0 -90 0>}]
 belt_01 [{
   shapeType=Capsule
   length=6.8628
   radius=3.4746
   shapeName=belt_01_capsule
   center=<7.0798 -0.7263 1.2486>
   rotation=<rot 0.0001 90.2464 -100.1763>
  }]
 belt_02 [{
   shapeType=Capsule
   length=7.5301
   radius=3.8285
   shapeName=belt_02_capsule
   center=<3.8276 -0.6068 2.0125>
   rotation=<rot 7.04 86.3674 -94.0983>
  }]
 belt_04 [{
   shapeType=Capsule
   length=1.4666
   radius=3.5686
   shapeName=belt_04_capsule
   center=<1.958 -1.2028 -1.0133>
   rotation=<rot 0 174.7006 -88.161>
  }]
 calf_l [{shapeType=Capsule, length=21.9801, radius=6.8373, center=<-30.9618 -2.9586 0.075>, rotation=<rot 0 -90.7571 -89.4577>}]
 calf_r [{shapeType=Capsule, length=21.9801, radius=6.8373, center=<30.9618 2.9586 -0.075>, rotation=<rot 0 -90.7571 90.5423>}]
 foot_l [{shapeType=Capsule, length=13.8182, radius=6.5955, center=<-6.1589 7.4761 0.037>, rotation=<rot 0 3.2204 -86.8667>}]
 foot_r [{shapeType=Capsule, length=13.8182, radius=6.5955, center=<6.1589 -7.4761 -0.037>, rotation=<rot 0 3.2204 93.1333>}]
 hand_l [{shapeType=Capsule, length=0, radius=5.4224, center=<7.8986 -2.564 -0.3949>, rotation=<rot 0 82.5684 -82.3961>}]
 hand_r [{shapeType=Capsule, length=9.2608, radius=5.656, center=<-7.6065 2.5172 0.1986>, rotation=<rot 0 82.301 -80.0233>}]
 lowerarm_r [{shapeType=Capsule, length=13.876, radius=4.9358, center=<-9.9007 -0.0382 -0.032>, rotation=<rot 0 86.1211 -82.8098>}]
 pelvis [{shapeType=Capsule, length=15.4916, radius=9.5042, center=<-2.8895 -3.3705 1.5599>, rotation=<rot -5.8647 88.5978 -178.6116>}]
 spine_01 [{shapeType=Capsule, length=21.7485, radius=6.756, shapeName=spine_01_capsule, center=<2.4405 -1.049 0>, rotation=<rot 0 -90 0>}]
 spine_04 [{shapeType=Capsule, length=7.5979, radius=2.5, center=<-0.9183 2.3426 -0.0289>, rotation=<rot 0.0129 -90 -0.4078>}]
 spine_05 [{shapeType=Capsule, length=14.1862, radius=9.8324, center=<5.6528 0.0017 -0.5381>, rotation=<rot 11.9609 -90 -6.4179>}]
 thigh_twist_01_l [
  {
   shapeType=Capsule
   length=34.2323
   radius=7.3497
   shapeName=thigh_twist_01_l_capsule
   center=<9.141 -0.218 -0.7154>
   rotation=<rot -17.5986 -91.2245 -92.4067>
  }
  {
   shapeType=Capsule
   length=22.242
   radius=7.758
   shapeName=thigh_twist_01_l_capsule
   center=<7.9458 -5.3254 10.6159>
   rotation=<rot 88.0554 0.0001 0.0001>
  }
 ]
 thigh_twist_01_r [{
   shapeType=Capsule
   length=32.4751
   radius=9.107
   shapeName=thigh_twist_01_r_capsule
   center=<-8.8676 0.9264 0.7858>
   rotation=<rot -17.5458 -90.0001 87.2237>
  }]
 upperarm_l [{shapeType=Capsule, length=0, radius=0.5, center=<23.6478 3.4879 -0.6243>, rotation=<rot 0 104.1487 -85.7911>}]
 upperarm_r [{shapeType=Capsule, length=13.8608, radius=6.2421, center=<-12.9922 -0.2112 0.8998>, rotation=<rot 0 -93.0882 -83.0823>}]
}
constraints [
 {bone1Name=spine_04, bone2Name=spine_01, swing1LimitDegrees=45, swing2LimitDegrees=45, twistLimitDegrees=45}
 {bone1Name=spine_05, bone2Name=spine_04, swing1LimitDegrees=45, swing2LimitDegrees=45, twistLimitDegrees=45}
 {bone1Name=upperarm_l, bone2Name=spine_05, swing1LimitDegrees=45, swing2LimitDegrees=45, twistLimitDegrees=45}
 {bone1Name=hand_l, bone2Name=upperarm_l, swing1LimitDegrees=45, swing2LimitDegrees=45, twistLimitDegrees=45}
 {bone1Name=upperarm_r, bone2Name=spine_05, swing1LimitDegrees=45, swing2LimitDegrees=45, twistLimitDegrees=45}
 {bone1Name=lowerarm_r, bone2Name=upperarm_r, swing1LimitDegrees=45, swing2LimitDegrees=45, twistLimitDegrees=45}
 {bone1Name=hand_r, bone2Name=lowerarm_r, swing1LimitDegrees=45, swing2LimitDegrees=45, twistLimitDegrees=45}
 {bone1Name=calf_l, bone2Name=pelvis, swing1LimitDegrees=45, swing2LimitDegrees=45, twistLimitDegrees=45}
 {bone1Name=foot_l, bone2Name=calf_l, swing1LimitDegrees=45, swing2LimitDegrees=45, twistLimitDegrees=45}
 {bone1Name=calf_r, bone2Name=pelvis, swing1LimitDegrees=45, swing2LimitDegrees=45, twistLimitDegrees=45}
 {bone1Name=foot_r, bone2Name=calf_r, swing1LimitDegrees=45, swing2LimitDegrees=45, twistLimitDegrees=45}
 {
  bone1Name=based_belt_01
  bone2Name=pelvis
  swing1LimitDegrees=45
  swing1Motion=Locked
  swing2LimitDegrees=45
  swing2Motion=Locked
  twistLimitDegrees=45
  twistMotion=Locked
 }
 {bone1Name=belt_01, bone2Name=based_belt_01, swing1LimitDegrees=27, swing2LimitDegrees=22, twistLimitDegrees=6}
 {bone1Name=belt_02, bone2Name=belt_01, swing1LimitDegrees=15, swing2LimitDegrees=22, twistLimitDegrees=6}
 {
  bone1Name=thigh_twist_01_l
  bone2Name=pelvis
  swing1LimitDegrees=45
  swing1Motion=Free
  swing2LimitDegrees=45
  swing2Motion=Free
  twistLimitDegrees=45
  twistMotion=Free
 }
 {
  bone1Name=thigh_twist_01_r
  bone2Name=pelvis
  swing1LimitDegrees=45
  swing1Motion=Free
  swing2LimitDegrees=45
  swing2Motion=Free
  twistLimitDegrees=45
  twistMotion=Free
 }
 {
  bone1Name=spine_01
  bone2Name=pelvis
  swing1LimitDegrees=45
  swing1Motion=Free
  swing2LimitDegrees=45
  swing2Motion=Free
  twistLimitDegrees=45
  twistMotion=Free
 }
 {bone1Name=belt_04, bone2Name=belt_02, swing1LimitDegrees=2, swing2LimitDegrees=11, twistLimitDegrees=6}
]
