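# Animation Blueprint for the female medium FondAim player cosmetic body, adding secondary motion to its accessory bones over a pose copied from the attached parent mesh.
# Weighted bone constraints and a Control Rig fed the skydiving flag position spine-correction, clip, clasp and grenade bones.
# A pelvis-space RigidBody, a grenade Look At and chained AnimDynamics springs then drive bows, straps and grenades.
# Bow sim-space settings, chosen on init and customization load, lower angular alpha in the front end.
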AnimBlueprint /BRCosmetics/Characters/Player/Female/Medium/Bodies/F_MED_FondAim/Meshes/F_MED_FondAim_AnimBP
target_skeleton=&/F_MED_FondAim_Skeleton
graphs [
 AnimGraph {nodes [
   n1 = Get Sim Space Settings() -> (Sim Space Settings:Anim Phys Sim Space Settings Structure)
   n2 = "Constraint - correct_spine_base (PPPPPPPP)"!(
    ComponentPose=Local To Component(LocalPose=Copy Pose From Mesh() {bUseAttachedParent=true})
    Alpha=1
    ConstraintWeights_0=0.1328
    ConstraintWeights_1:float=0.1097
    ConstraintWeights_2:float=0.3703
    ConstraintWeights_3:float=0.0229
    ConstraintWeights_4:float=0.1277
    ConstraintWeights_5:float=0.1837
    ConstraintWeights_6:float=0.0495
    ConstraintWeights_7:float=0.0035
   ) -> (Pose) {
    boneToModify {boneName=correct_spine_base}
    constraintSetup [
     {targetBone {boneName=spine_05}}
     {targetBone {boneName=clavicle_l}}
     {targetBone {boneName=spine_04}}
     {targetBone {boneName=upperarm_twist_01_l}}
     {targetBone {boneName=spine_03}}
     {targetBone {boneName=spine_02}}
     {targetBone {boneName=spine_01}}
     {targetBone {boneName=pelvis}}
    ]
    constraintWeights [1,1,1,1,1,1,1,1]
   }
   n3 = "Constraint - clips (PPPPPPPP)"!(
    ComponentPose=Local To Component(LocalPose=Component To Local(ComponentPose=@n2))
    Alpha=1
    ConstraintWeights_0=0.2585
    ConstraintWeights_1:float=0.2801
    ConstraintWeights_2:float=0.3457
    ConstraintWeights_3:float=0.0192
    ConstraintWeights_4:float=0.0205
    ConstraintWeights_5:float=0.0397
    ConstraintWeights_6:float=0.0209
    ConstraintWeights_7:float=0.0154
   ) -> (Pose) {
    boneToModify {boneName=clips}
    constraintSetup [
     {targetBone {boneName=neck_02}}
     {targetBone {boneName=neck_01}}
     {targetBone {boneName=spine_05}}
     {targetBone {boneName=FACIAL_L_Jawline}}
     {targetBone {boneName=FACIAL_R_UnderChin}}
     {targetBone {boneName=FACIAL_C_UnderChin}}
     {targetBone {boneName=FACIAL_L_UnderChin}}
     {targetBone {boneName=FACIAL_R_Jawline}}
    ]
    constraintWeights [1,1,1,1,1,1,1,1]
   }
   n4 = Control Rig!(Source=Component To Local(ComponentPose=@n3), Alpha, "sky ":bool=Get bIsSkyDiving()) -> (Pose) {controlRigAssetReference {blueprintRigClass=&/F_MED_FondAim_CtrlRig.F_MED_FondAim_CtrlRig_C}}
   n5 = "Constraint - based_l_clasp (PPPPPPPP)"!(
    ComponentPose=Local To Component(LocalPose=@n4)
    Alpha=1
    ConstraintWeights_0=0.2505
    ConstraintWeights_1:float=0.382
    ConstraintWeights_2:float=0.1107
    ConstraintWeights_3:float=0.091
    ConstraintWeights_4:float=0.0001
    ConstraintWeights_5:float=0.0841
    ConstraintWeights_6:float=0.0179
    ConstraintWeights_7:float=0.0638
   ) -> (Pose) {
    boneToModify {boneName=based_l_clasp}
    constraintSetup [
     {targetBone {boneName=spine_05}}
     {targetBone {boneName=clavicle_l}}
     {targetBone {boneName=spine_04}}
     {targetBone {boneName=upperarm_twist_01_l}}
     {targetBone {boneName=spine_03}}
     {targetBone {boneName=deform_upperarm_fr_l}}
     {targetBone {boneName=deform_upperarm_up_l}}
     {targetBone {boneName=deform_pec_l}}
    ]
    constraintWeights [1,1,1,1,1,1,1,1]
   }
   n6 = "Constraint - based_r_clasp (PPPPPPPP)"!(
    ComponentPose=@n5
    Alpha=1
    ConstraintWeights_0=0.2054
    ConstraintWeights_1:float=0.3411
    ConstraintWeights_2:float=0.1066
    ConstraintWeights_3:float=0.1144
    ConstraintWeights_4:float=0.0132
    ConstraintWeights_5:float=0.1182
    ConstraintWeights_6:float=0.0225
    ConstraintWeights_7:float=0.0787
   ) -> (Pose) {
    boneToModify {boneName=based_r_clasp}
    constraintSetup [
     {targetBone {boneName=spine_05}}
     {targetBone {boneName=clavicle_r}}
     {targetBone {boneName=spine_04}}
     {targetBone {boneName=upperarm_twist_01_r}}
     {targetBone {boneName=pelvis}}
     {targetBone {boneName=deform_upperarm_fr_r}}
     {targetBone {boneName=deform_upperarm_up_r}}
     {targetBone {boneName=deform_pec_r}}
    ]
    constraintWeights [1,1,1,1,1,1,1,1]
   }
   n7 = "Constraint - based_grenade_03 (PPP)"!(ComponentPose=@n6, Alpha=1, ConstraintWeights_0=0.5799, ConstraintWeights_1:float=0.2398, ConstraintWeights_2:float=0.1803) -> (Pose) {
    boneToModify {boneName=based_grenade_03}
    constraintSetup [{targetBone {boneName=pelvis}}, {targetBone {boneName=thigh_twist_01_r}}, {targetBone {boneName=deform_glute_r}}]
    constraintWeights [1,1,1]
   }
   n8 = "Constraint - based_grenade_02 (PPPPP)"!(
    ComponentPose=@n7
    Alpha=1
    ConstraintWeights_0=0.4132
    ConstraintWeights_1:float=0.4544
    ConstraintWeights_2:float=0.0189
    ConstraintWeights_3:float=0.0415
    ConstraintWeights_4:float=0.072
   ) -> (Pose) {
    boneToModify {boneName=based_grenade_02}
    constraintSetup [
     {targetBone {boneName=pelvis}}
     {targetBone {boneName=thigh_twist_01_r}}
     {targetBone {boneName=thigh_r}}
     {targetBone {boneName=deform_glute_r}}
     {targetBone {boneName=deform_groin_r}}
    ]
    constraintWeights [1,1,1,1,1]
   }
   n9 = "Constraint - based_grenade_01 (PPPPP)"!(
    ComponentPose=@n8
    Alpha=1
    ConstraintWeights_0=0.0114
    ConstraintWeights_1:float=0.5173
    ConstraintWeights_2:float=0.3082
    ConstraintWeights_3:float=0.0144
    ConstraintWeights_4:float=0.1487
   ) -> (Pose) {
    boneToModify {boneName=based_grenade_01}
    constraintSetup [
     {targetBone {boneName=spine_02}}
     {targetBone {boneName=pelvis}}
     {targetBone {boneName=thigh_twist_01_r}}
     {targetBone {boneName=thigh_r}}
     {targetBone {boneName=deform_groin_r}}
    ]
    constraintWeights [1,1,1,1,1]
   }
   n10 = RigidBody(ComponentPose=@n9) {
    overridePhysicsAsset=&/F_MED_FondAim_Physics
    simulationSpace=BaseBoneSpace
    baseBoneRef {boneName=pelvis}
    componentAppliedLinearAccClamp=<3333 3333 3333>
    simSpaceSettings {
     dampingAlpha=0
     maxAngularAcceleration=12
     maxAngularVelocity=3
     maxLinearAcceleration=1550
     maxLinearVelocity=600
     worldAlpha=0.4832
     externalLinearDragV=<1 1 1>
    }
   }
   n11 = Look At - Bone: based_grenade_01(ComponentPose=@n10, Alpha=1) {
    boneToModify {boneName=based_grenade_01}
    lookAtLocation=<4.918 -21.3447 30.8262>
    lookAtTarget {boneReference {boneName=thigh_r}}
    lookAt_Axis {axis=<1 0>}
    lookUp_Axis {axis {x=0, z=1}}
   }
   n12 = Anim Dynamics - Bone: based_bow_01!(ComponentPose=@n11, Alpha=1, SimSpaceSettings:Anim Phys Sim Space Settings Structure=@n1, ExternalForce="0,0,0") -> (Pose) {
    angularDampingOverride=0.9999
    angularSpringConstant=1000
    bAngularSpring=true
    bGravityOverrideInSimSpace=true
    bOverrideAngularDamping=true
    bUseGravityOverride=true
    simulationSpace=BoneRelative
    alphaBoolBlend {blendOption=Linear}
    boundBone {boneName=based_bow_01}
    componentAppliedLinearAccClamp=<5000 5000 5000>
    componentLinearAccScale=<0.3 0.3 0.3>
    componentLinearVelScale=<0.3 0.3 0.3>
    gravityOverride {x=-980}
    physicsBodyDefinitions [{
      boundBone {boneName=based_bow_01}
      boxExtents=<16 16 16>
      constraintSetup {angularTargetAxis=AxisZ, angularLimitsMax=<22 5 2>, angularLimitsMin=<-10 -22 -2>, angularTarget=<0 0 1>}
      localJointOffset=<0 0 1>
     }]
    relativeSpaceBone {boneName=upperarm_r}
   }
   n13 = Anim Dynamics - Bone: based_bow_02!(ComponentPose=@n12, Alpha=1, SimSpaceSettings:Anim Phys Sim Space Settings Structure=@n1, ExternalForce="0,0,0") -> (Pose) {
    angularDampingOverride=0.9999
    angularSpringConstant=1000
    bAngularSpring=true
    bGravityOverrideInSimSpace=true
    bOverrideAngularDamping=true
    bUseGravityOverride=true
    simulationSpace=BoneRelative
    alphaBoolBlend {blendOption=Linear}
    boundBone {boneName=based_bow_02}
    componentAppliedLinearAccClamp=<5000 5000 5000>
    componentLinearAccScale=<0.3 0.3 0.3>
    componentLinearVelScale=<0.3 0.3 0.3>
    gravityOverride {x=-980}
    physicsBodyDefinitions [{
      boundBone {boneName=based_bow_02}
      boxExtents=<16 16 16>
      constraintSetup {angularTargetAxis=AxisZ, angularLimitsMax=<22 5 2>, angularLimitsMin=<-10 -22 -2>, angularTarget=<0 0 1>}
      localJointOffset=<0 0 1>
     }]
    relativeSpaceBone {boneName=upperarm_r}
    simSpaceSettings {maxAngularAcceleration=600, maxAngularVelocity=150, simSpaceAngularAlpha=1}
   }
   n14 = Anim Dynamics - Bone: based_strap_hend_02(ComponentPose=@n13, Alpha=1, ExternalForce="0,0,0") {
    angularDampingOverride=0.999
    angularSpringConstant=8000
    bAngularSpring=true
    bOverrideAngularDamping=true
    bUseGravityOverride=true
    simulationSpace=BoneRelative
    alphaBoolBlend {blendOption=Linear}
    boundBone {boneName=based_strap_hend_02}
    componentAppliedLinearAccClamp=<5000 5000 5000>
    componentLinearAccScale=<0.3 0.3 0.3>
    componentLinearVelScale=<0.3 0.3 0.3>
    gravityOverride {z=-980}
    physicsBodyDefinitions [{
      boundBone {boneName=based_strap_hend_02}
      boxExtents=<11 11 11>
      constraintSetup {angularTargetAxis=AxisZ, twistAxis=AxisZ, angularLimitsMax=<22 10 10>, angularLimitsMin=<-44 -10 -10>, angularTarget=<1 0 0>}
      localJointOffset=<0 0 1>
     }]
    relativeSpaceBone {boneName=hand_r}
   }
   n15 = Anim Dynamics - Bone: shoe_strap_l_01(ComponentPose=@n14, Alpha=1, ExternalForce="0,0,0") {
    angularDampingOverride=0.999
    angularSpringConstant=8000
    bAngularSpring=true
    bOverrideAngularDamping=true
    bUseGravityOverride=true
    simulationSpace=BoneRelative
    alphaBoolBlend {blendOption=Linear}
    boundBone {boneName=shoe_strap_l_01}
    componentAppliedLinearAccClamp=<5000 5000 5000>
    componentLinearAccScale=<0.3 0.3 0.3>
    componentLinearVelScale=<0.3 0.3 0.3>
    gravityOverride {z=-980}
    physicsBodyDefinitions [{
      boundBone {boneName=shoe_strap_l_01}
      boxExtents=<11 11 11>
      constraintSetup {angularTargetAxis=AxisZ, twistAxis=AxisZ, angularLimitsMax=<22 10 10>, angularLimitsMin=<-55 -10 -10>, angularTarget=<1 0 0>}
      localJointOffset=<0 0 1.25>
     }]
    relativeSpaceBone {boneName=thigh_l}
   }
   n16 = Anim Dynamics - Bone: based_shoe_strap_r_top(ComponentPose=@n15, Alpha=1, ExternalForce="0,0,0") {
    angularDampingOverride=0.999
    angularSpringConstant=8000
    bAngularSpring=true
    bOverrideAngularDamping=true
    bUseGravityOverride=true
    simulationSpace=BoneRelative
    alphaBoolBlend {blendOption=Linear}
    boundBone {boneName=based_shoe_strap_r_top}
    componentAppliedLinearAccClamp=<5000 5000 5000>
    componentLinearAccScale=<0.3 0.3 0.3>
    componentLinearVelScale=<0.3 0.3 0.3>
    gravityOverride {z=-980}
    physicsBodyDefinitions [{
      boundBone {boneName=based_shoe_strap_r_top}
      boxExtents=<9 9 9>
      constraintSetup {angularTargetAxis=AxisZ, twistAxis=AxisZ, angularLimitsMax=<22 10 10>, angularLimitsMin=<-55 -10 -10>, angularTarget=<1 0 0>}
      localJointOffset=<0 0 1.25>
     }]
    relativeSpaceBone {boneName=thigh_r}
   }
   n17 = Anim Dynamics - Bone: shoe_strap_r_down_01(ComponentPose=@n16, Alpha=1, ExternalForce="0,0,0") {
    angularDampingOverride=0.999
    angularSpringConstant=8000
    bAngularSpring=true
    bOverrideAngularDamping=true
    bUseGravityOverride=true
    simulationSpace=BoneRelative
    alphaBoolBlend {blendOption=Linear}
    boundBone {boneName=shoe_strap_r_down_01}
    componentAppliedLinearAccClamp=<5000 5000 5000>
    componentLinearAccScale=<0.3 0.3 0.3>
    componentLinearVelScale=<0.3 0.3 0.3>
    gravityOverride {z=-980}
    physicsBodyDefinitions [{
      boundBone {boneName=shoe_strap_r_down_01}
      boxExtents=<11 11 11>
      constraintSetup {angularTargetAxis=AxisZ, twistAxis=AxisZ, angularLimitsMax=<22 10 10>, angularLimitsMin=<-55 -10 -10>, angularTarget=<1 0 0>}
      localJointOffset=<0 0 1.25>
     }]
    relativeSpaceBone {boneName=thigh_r}
   }
   n18 = Anim Dynamics - Bone: grenade_down_01(ComponentPose=@n17, Alpha=1, ExternalForce="0,0,0") {
    angularDampingOverride=0.999
    angularSpringConstant=200
    bAngularSpring=true
    bOverrideAngularDamping=true
    bUseGravityOverride=true
    simulationSpace=RootRelative
    boundBone {boneName=grenade_down_01}
    componentAppliedLinearAccClamp=<3000 3000 3000>
    componentLinearAccScale=<0.3 0.3 0.35>
    componentLinearVelScale=<0.3 0.3 0.35>
    gravityOverride {z=-980}
    physicsBodyDefinitions [{
      boundBone {boneName=grenade_down_01}
      boxExtents=<13 13 13>
      constraintSetup {angularLimitsMax=<12 0 5>, angularLimitsMin=<-12 -30 -15>, angularTarget=<1 0 0>}
      localJointOffset=<2 0 0>
     }]
    simSpaceSettings {maxAngularAcceleration=35, maxAngularVelocity=7, simSpaceAngularAlpha=0.3}
   }
   n19 = Anim Dynamics - Bone: grenade_right_01(ComponentPose=@n18, Alpha=1, ExternalForce="0,0,0") {
    angularDampingOverride=0.999
    angularSpringConstant=200
    bAngularSpring=true
    bOverrideAngularDamping=true
    bUseGravityOverride=true
    simulationSpace=RootRelative
    boundBone {boneName=grenade_right_01}
    componentAppliedLinearAccClamp=<3000 3000 3000>
    componentLinearAccScale=<0.3 0.3 0.35>
    componentLinearVelScale=<0.3 0.3 0.35>
    gravityOverride {z=-980}
    physicsBodyDefinitions [{
      boundBone {boneName=grenade_right_01}
      boxExtents=<13 13 13>
      constraintSetup {angularLimitsMax=<12 1 5>, angularLimitsMin=<-12 -30 -11>, angularTarget=<1 0 0>}
      localJointOffset=<2 0 0>
     }]
    simSpaceSettings {maxAngularAcceleration=35, maxAngularVelocity=7, simSpaceAngularAlpha=0.3}
   }
   n20 = Anim Dynamics - Bone: grenade_top_01(ComponentPose=@n19, Alpha=1, ExternalForce="0,0,0") {
    angularDampingOverride=0.8
    angularSpringConstant=500
    bAngularSpring=true
    bOverrideAngularDamping=true
    bUseGravityOverride=true
    simulationSpace=RootRelative
    worldSpaceMinimumScale=0.1
    boundBone {boneName=grenade_top_01}
    componentAppliedLinearAccClamp=<3000 3000 3000>
    componentLinearAccScale=<0.3 0.3 0.35>
    componentLinearVelScale=<0.3 0.3 0.35>
    gravityOverride {z=-980}
    physicsBodyDefinitions [{
      boundBone {boneName=grenade_top_01}
      boxExtents=<9 9 9>
      constraintSetup {angularLimitsMax=<12 13 0>, angularLimitsMin=<-12 -20 -15>, angularTarget=<1 0 0>}
      localJointOffset=<0.4 0 0>
     }]
    simSpaceSettings {maxAngularAcceleration=35, maxAngularVelocity=7, simSpaceAngularAlpha=0.3}
   }
   Output Pose(Result=Component To Local(ComponentPose=@n20)) {layerGroup=None}
  ]}
 EventGraph {nodes [
   n1 = Event BlueprintInitializeAnimation()
   Set Sim Space Settings(
    execute=[@n1.then, @n2.then]
    Sim Space Settings:Anim Phys Sim Space Settings Structure=Select(
     Option 0:Anim Phys Sim Space Settings Structure=Make Anim Phys Sim Space Settings!(SimSpaceAngularAlpha:float=0.7, MaxAngularVelocity:float=300, MaxAngularAcceleration:float=1500, ExternalAngularVelocity:Vector="0,0,0") -> (AnimPhysSimSpaceSettings:Anim Phys Sim Space Settings Structure) {is=title K2Node_MakeStruct}
     Option 1:Anim Phys Sim Space Settings Structure=Make Anim Phys Sim Space Settings(SimSpaceAngularAlpha=0.3, MaxAngularVelocity=300, MaxAngularAcceleration=1500, ExternalAngularVelocity="0,0,0")
     Index=Get bIsInFrontEnd()
    ) -> (ReturnValue:Anim Phys Sim Space Settings Structure)
   ) -> (Output_Get:Anim Phys Sim Space Settings Structure)
   n2 = Event OnCharacterCustomizationLoaded()
   Event BlueprintUpdateAnimation()
  ]}
]
variables [Sim Space Settings {
  type=AnimPhysSimSpaceSettings
  default {maxAngularAcceleration=600, maxAngularVelocity=150, simSpaceAngularAlpha=0.3, externalAngularVelocity=<0 0 0>}
 }]
