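# Physics asset for the FondAim medium female player body that, per its name, supplies collision shapes for the cape.
# Capsules cover spine, pelvis, upper and lower arms, hands, thighs, calves and feet, with an extra sphere on the pelvis.
# It defines no constraints between bodies, so it is a set of bone-attached colliders rather than a jointed ragdoll.
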PhysicsAsset /BRCosmetics/Characters/Player/Female/Medium/Bodies/F_MED_FondAim/Meshes/M_MED_Fondaim_CapeColliders
bodies {
 calf_l [{shapeType=Capsule, length=32.5143, radius=15.6454, center=<-23.0006 0.4333 0>, rotation=<rot 90 0 0>}]
 calf_r [{shapeType=Capsule, length=32.5143, radius=15.6454, center=<23.0006 -1.7468 0>, rotation=<rot 90 180 0>}]
 foot_l [{shapeType=Capsule, length=10.5461, radius=14.1734, center=<-4.1955 6.0815 0>, rotation=<rot 0 14.9998 90>}]
 foot_r [{shapeType=Capsule, length=10.5461, radius=14.1734, center=<4.1955 -6.0815 0>, rotation=<rot 0 14.9998 -90>}]
 hand_l [{shapeType=Capsule, length=9.5015, radius=6.7933, center=<6.2703 0 0>, rotation=<rot 90 0 0>}]
 hand_r [{shapeType=Capsule, length=9.5015, radius=6.7933, center=<-6.2703 0 0>, rotation=<rot 90 0 180>}]
 lowerarm_l [{shapeType=Capsule, length=14.3387, radius=7.5574, center=<8.9852 0 0.2267>, rotation=<rot 90 0 0>}]
 lowerarm_r [{shapeType=Capsule, length=14.3387, radius=7.5574, center=<-8.9852 0 -0.2267>, rotation=<rot 90 0 180>}]
 pelvis [
  {shapeType=Sphere, radius=6.1725, shapeName=pelvis_sphere, center=<-8.237 -12.9854 15.9288>}
  {shapeType=Capsule, length=19.8434, radius=17.7874, center=<-13.7107 3.9218 0.5602>}
 ]
 spine_02 [{shapeType=Capsule, length=0.7, radius=18.0784, center=<-2.6142 9.6134 0>}]
 spine_03 [{shapeType=Capsule, length=3.4972, radius=16.2226, center=<0 7.2156 0>}]
 spine_04 [{shapeType=Capsule, length=2.5255, radius=17.3781, center=<0 9.1222 0>}]
 spine_05 [{shapeType=Capsule, length=17.473, radius=12.4488, center=<0 3.1057 0>}]
 thigh_l [{shapeType=Capsule, length=15.4697, radius=14.5854, center=<-18.1899 3.9453 0>, rotation=<rot 90 177.336 0>}]
 thigh_r [{shapeType=Capsule, length=15.4697, radius=14.5854, center=<18.1899 -3.9453 0>, rotation=<rot 90 -2.664 0>}]
 upperarm_l [{shapeType=Capsule, length=17.1712, radius=7.7501, center=<10.0455 0 0>, rotation=<rot 90 0 0>}]
 upperarm_r [{shapeType=Capsule, length=17.1712, radius=7.7501, center=<-10.0455 0 0>, rotation=<rot 90 0 180>}]
}
constraints []
